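# Parent material for the SameRabbit player character body, adding an antenna light to the shared 2023 character shading.
# Material attributes from the PreFX function are modified through named reroutes, alongside a PostFX function call.
# The light multiplies desaturated base color by AntenaLightColor and blends it over emissive using a power-shaped channel mask.
# EnableEffect scales that blend and defaults to zero, so the light is off unless overridden.
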Material /Game/Characters/Player/Female/Medium/Bodies/F_MED_SameRabbit/Materials/M_SameRabbit_Parent
graph {
 properties {
  bUseMaterialAttributes=true
  bUsedWithClothing=true
  bUsedWithMeshDeformer=true
  bUsedWithMorphTargets=true
  bUsedWithNanite=true
  bUsedWithNiagaraMeshParticles=true
  bUsedWithSkeletalMesh=true
  pixelDepthOffsetMode=PDOM_Legacy
  shadingModel=MSM_FromMaterialExpression
  usedShadingModels=Default Lit | Subsurface | Clear Coat | Subsurface Profile | Cloth
 }
 nodes [
  n1 = MaterialFunctionCall() {materialFunction=&MF_CharacterParent_2023_PreFX}
  MaterialFunctionCall(Material=NamedRerouteUsage(), M=@n1.M, Pre Skinned Local Position=@n1.M, Distance Blend=@n1.M, Fresnel=@n1.M) {materialFunction=&MF_CharacterParent_2023_PostFX}
  NamedRerouteDeclaration(None=@n1.Result) {name=InMaterialAttributes, nodeColor=<rgba 0 0.7891 1 1>}
  NamedRerouteDeclaration(None=@n1.Tech Art Mask) {name=InFXMask, nodeColor=<rgba 1 0.1875 0 1>}
  n2 = GetMaterialAttributes(None=NamedRerouteUsage()) {attributeGetTypes [#1, #2]}
  n3 = Multiply(
   A=Desaturation(None=@n2.Base Color, Fraction=Constant() {r=1})
   B=VectorParameter() {group=SameRabbit, parameterName=AntenaLightColor, defaultValue=<rgba 1 0.9536 0.125 1>}.RGB
  )
  n4 = ChannelMaskParameter(None=NamedRerouteUsage()) {group=SameRabbit, parameterName=AntenaMask}
  n5 = ScalarParameter() {defaultValue=2, group=SameRabbit, parameterName=AntenaMaskSoftness}
  n6 = Multiply(A=Saturate(None=Power(Base=@n4.R, Exp=@n5)), B=ScalarParameter() {group=SameRabbit, parameterName=EnableEffect, sliderMax=1})
  NamedRerouteDeclaration(None=SetMaterialAttributes(MaterialAttributes=@n2.MaterialAttributes, Emissive Color=LinearInterpolate(A=@n2.Emissive Color, B=@n3, Alpha=@n6)) {attributeSetTypes [#2]}) {name=OutMaterialAttributes, nodeColor=<rgba 0 1 0.1562 1>}
 ]
}
